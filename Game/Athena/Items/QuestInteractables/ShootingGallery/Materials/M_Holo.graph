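# M_Holo is an additive, two-sided material in the Shooting Gallery quest interactable folder, apparently for a hologram look, previewed on a simple target panel mesh.
# It builds a Fresnel-like edge mask from the inverted absolute Z of the tangent-space camera vector, branching into a Clamp and a Power.
# A blue color constant and an empty Multiply sit unconnected, and no links to material outputs appear.
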Material /Game/Athena/Items/QuestInteractables/ShootingGallery/Materials/M_Holo
graph {
 properties {
  blendMode=BLEND_Additive
  pixelDepthOffsetMode=PDOM_Legacy
  previewMesh=/Game/Athena/Items/QuestInteractables/ShootingGallery/Meshes/SM_SimpleTargetPanel
  twoSided=true
  usedShadingModels=Default Lit
 }
 nodes [
  n1 = Transform(None=CameraVectorWS()) {transformSourceType=TRANSFORMSOURCE_World, transformType=TRANSFORM_Tangent}
  n2 = OneMinus(None=Abs(None=ComponentMask(None=@n1) {b=true}))
  Clamp(None=@n2)
  Constant3Vector() {constant=<rgba 0.045 0.1922 1 1>}
  Multiply()
  Power(Base=@n2) {constExponent=16}
 ]
}
